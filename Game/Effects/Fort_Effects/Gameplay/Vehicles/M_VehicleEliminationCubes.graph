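# Niagara mesh-particle material that gives vehicle elimination cubes a glowing cyan-blue look, tinted by particle color.
# Fresnel shapes the color gradient, with extra rim shaping at Epic quality.
# A dynamic parameter offsets scaled UVs into a grid-squares texture (Cube Color U feeds both axes; V is unused), driving a time-based sine flicker.
# Epic quality adds this flicker and an LDR-to-HDR remap; Default uses just the tinted gradient.
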Material /Game/Effects/Fort_Effects/Gameplay/Vehicles/M_VehicleEliminationCubes
graph {
 properties {bUsedWithNiagaraMeshParticles=true, pixelDepthOffsetMode=PDOM_Legacy, usedShadingModels=Default Lit}
 nodes [
  Constant()
  Constant() {r=1}
  Constant() {r=0.2}
  n1 = Fresnel() {baseReflectFraction=0, exponent=1}
  n2 = Clamp(None=@n1)
  n3 = Power(Base=OneMinus(None=@n2)) {constExponent=4}
  n4 = LinearInterpolate(Alpha=Power(Base=@n2)) {constA=0.3, constB=0.9}
  n5 = QualitySwitch(Default=@n1, Epic=Multiply(A=Add(A=@n3, B=@n4)) {constB=0.6})
  n6 = Add(
   A=Constant3Vector() {constant=<rgba 0.0583 0.3109 0.4 1>}
   B=Multiply(A=Power(Base=@n5) {constExponent=3}, B=Constant3Vector() {constant=<rgba 0 3.8199 6 1>})
  )
  n7 = ParticleColor()
  n8 = Power(Base=@n6)
  n9 = TextureCoordinate()
  n10 = Constant() {r=0.0625}
  n11 = DynamicParameter() {defaultValue=<rgba 0 1 1 1>, paramNames [Cube Color U, Cube Color V, Param3, Param4]}
  n12 = TextureSample(UVs=Add(A=Multiply(A=@n9, B=@n10), B=AppendVector(A=@n11.Cube Color U, B=@n11.Cube Color U))) {
   samplerSource=SSM_Wrap_WorldGroupSettings
   samplerType=SAMPLERTYPE_LinearColor
   texture=/Game/Packages/Fortress_SharedMaterials/Textures/T_Slowfield_GridSquares_01
  }
  n13 = Add(A=@n12.R, B=Time())
  n14 = Multiply(A=ConstantBiasScale(None=Sine(None=@n13)), B=@n12.R)
  n15 = Multiply(A=Add(A=Multiply(A=@n6) {constB=0.3}, B=Multiply(A=@n8, B=@n14)), B=@n7.RGB)
  QualitySwitch(
   Default=Multiply(A=@n6, B=@n7.RGB)
   Epic=MaterialFunctionCall(LDR Input=Clamp(None=@n15), HDR Tint=@n15) {materialFunction=&DeriveHDRfromLDR}.RemappedRange
  )
 ]
}
