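# Niagara GPU burst for a vehicle transition, made of blue digital noise and streaks.
# All three emitters spawn on a flat owner-scaled box offset by a system Range Control value.
# Two noise emitters mix mesh and sprite particles by spawn group; the WorldSpace one adds gravity and age-gated collision, dropping noise force after impact.
# Streaks accelerate along negative X while their spawn rate tapers over the loop.
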NiagaraSystem /Game/Effects/Fort_Effects/Gameplay/Vehicles/NS_Vehicle_Transition_Burst_v2
dependencies {
 usedDataInterfaces [&NiagaraDataInterfaceCurve]
 usedDynamicInputs [&Lerp_Vector]
 usedModules [
  &SystemState
  &self:SystemSpawnScript.SystemScriptSource.SystemScriptGraph.NiagaraNodeAssignment_1.SetVariables_#1
 ]
}
emitters {
 Digital_Noise {
  simTarget=GPUComputeSim
  emitterUpdateScript [
   &EmitterState()
   &SpawnRate(SpawnRate=750, Use Spawn Probability=false)
   &SpawnRate(SpawnGroup=1, SpawnRate=150, Use Spawn Probability=false) {module=SpawnRate001}
  ]
  particleSpawnScript [
   /SetVariables(Particles.VisibilityTag=&SpawnGroup()) {isSetParameters=true, module=SetVariables_#2}
   &InitializeParticle(
    Color=&MaskLinearColorBySpawnGroup(Masked Linear Color=<rgba 0.0813 0.5532 1 1>, Passthrough Linear Color=<rgba 0 5.1365 10 1>)
    Lifetime Max=0.5
    Lifetime Min=0.25
    Lifetime Mode=::NewEnumerator1
    Mesh Scale Mode=::NewEnumerator1
    Mesh Uniform Scale Max=3
    Mesh Uniform Scale Min=1.5
    Sprite Size Mode=::NewEnumerator3
    Uniform Sprite Size=&MaskFloatBySpawnGroup(Masked Float=14, Passthrough Float=&RandomRangeFloat(Minimum=6, Maximum=14))
    UsePositionOffset=false
    Write Material Random=true
   )
   /SetVariables(Particles.Flare Size:Vector2f=&Vector2DFromFloat(Value=&MaskFloatBySpawnGroup(Masked Float=14, Passthrough Float=20))) {isSetParameters=true, module=SetVariables_#3}
   &ShapeLocation(
    Box Midpoint=<0 0.5 0.5>
    Box Size=<0 250 100>
    MaskSpawnGroup=false
    Offset=$System.Range Control
    Offset Mode=::NewEnumerator0
    Override Seed=false
    Shape Primitive=::NewEnumerator2
    Surface Expansion Mode:ENiagaraExpansionMode=::NewEnumerator2
    Surface Only=true
    UseOwnerScale=true
   )
   &AddVelocity(
    Module.Use Falloff Distance=false
    Origin Offset=$Emitter.RangeControl
    Override Seed=false
    Velocity Mode=::NewEnumerator1
    Velocity Speed=150
   )
   &CameraOffset(Camera Offset Amount=&RandomRangeFloat(Minimum=25, Maximum=100))
  ]
  particleUpdateScript [
   &ParticleState()
   /SetVariables(Particles.SpriteColor=&MultiplyLinearColor(
     Linear Color=<rgba 0 0.5809 1 0.1>
     Scale Factor=&MakeVector4(
      X=1
      Y=1
      Z=1
      W=&FloatFromCurve(FloatCurve=DataInterface() {propertyValues {Curve {keys [{interpMode=RCIM_Cubic, tangentMode=RCTM_Break, time=0.3, value=1}, {interpMode=RCIM_Cubic, tangentMode=RCTM_Break, time=1, value=0}]}}})
     )
    )) {isSetParameters=true, module=SetVariables_#4}
   &ScaleColor(
    Scale Alpha=&FloatFromCurve(FloatCurve=DataInterface() {propertyValues {Curve {keys [
         {interpMode=RCIM_Cubic, leaveTangentWeight=0.883, tangentMode=RCTM_Break, time=0.3, value=1}
         {interpMode=RCIM_Cubic, tangentMode=RCTM_Break, time=1, value=0}
        ]}}})
    ScaleA=true
    ScaleRGB=true
   )
   &Drag(Drag=&RandomRangeFloat(Minimum=2, Maximum=4), Use Linear Drag=true, Use Rotational Drag=true)
   &VectorNoiseForce(Force Amount=2000, Module.Use Random Noise Time Interval=false)
   &SolveForcesAndVelocity(Clamp Velocity=false, Limit Acceleration=false)
   &DynamicMaterialParameters(
    Index 0 Param 1=0.9
    Module.Index0Param0Write=true
    Module.Index0Param1Write=false
    Module.Index0Param2Write=false
    Module.Index0Param3Write=false
   )
  ]
  rendererClasses [&NiagaraMeshRendererProperties, &NiagaraSpriteRendererProperties]
  renderers [
   {rendererClass=&NiagaraMeshRendererProperties, rendererIndex=0}
   {rendererClass=&NiagaraSpriteRendererProperties, rendererIndex=1}
   {rendererClass=&NiagaraSpriteRendererProperties, rendererIndex=2}
   {rendererClass=&NiagaraSpriteRendererProperties, rendererIndex=3}
  ]
 }
 Digital_Noise_WorldSpace {
  simTarget=GPUComputeSim
  emitterUpdateScript [
   &EmitterState()
   &SpawnRate(SpawnRate=250, Use Spawn Probability=false)
   &SpawnRate(SpawnGroup=1, SpawnRate=50, Use Spawn Probability=false) {module=SpawnRate001}
  ]
  particleSpawnScript [
   /SetVariables(Particles.VisibilityTag=&SpawnGroup()) {isSetParameters=true, module=SetVariables_#2}
   &InitializeParticle(
    Color=&MaskLinearColorBySpawnGroup(Masked Linear Color=<rgba 0.0813 0.5532 1 1>, Passthrough Linear Color=<rgba 0 5.1365 10 1>)
    Lifetime Max=0.5
    Lifetime Min=0.25
    Lifetime Mode=::NewEnumerator1
    Mesh Scale Mode=::NewEnumerator1
    Mesh Uniform Scale Max=3
    Mesh Uniform Scale Min=1.5
    Sprite Size Mode=::NewEnumerator3
    Uniform Sprite Size=&MaskFloatBySpawnGroup(Masked Float=14, Passthrough Float=&RandomRangeFloat(Minimum=6, Maximum=14))
    UsePositionOffset=false
    Write Material Random=true
   )
   /SetVariables(Particles.Flare Size:Vector2f=&Vector2DFromFloat(Value=&MaskFloatBySpawnGroup(Masked Float=14, Passthrough Float=20))) {isSetParameters=true, module=SetVariables_#3}
   &ShapeLocation(
    Box Midpoint=<0 0.5 0.5>
    Box Size=<0 250 100>
    MaskSpawnGroup=false
    Offset=$System.Range Control
    Offset Coordinate Space=::Local
    Offset Mode=::NewEnumerator0
    Override Seed=false
    Shape Primitive=::NewEnumerator2
    Surface Expansion Mode:ENiagaraExpansionMode=::NewEnumerator2
    Surface Only=true
    UseOwnerScale=true
   )
   &AddVelocity(
    Module.Use Falloff Distance=false
    Origin Offset=$System.Range Control
    Override Seed=false
    Velocity Mode=::NewEnumerator1
    Velocity Speed=150
   )
   &CameraOffset(Camera Offset Amount=&RandomRangeFloat(Minimum=25, Maximum=100))
  ]
  particleUpdateScript [
   &ParticleState()
   /SetVariables(Particles.SpriteColor=&MultiplyLinearColor(
     Linear Color=<rgba 0 0.5809 1 0.1>
     Scale Factor=&MakeVector4(
      X=1
      Y=1
      Z=1
      W=&FloatFromCurve(FloatCurve=DataInterface() {propertyValues {Curve {keys [{interpMode=RCIM_Cubic, tangentMode=RCTM_Break, time=0.3, value=1}, {interpMode=RCIM_Cubic, tangentMode=RCTM_Break, time=1, value=0}]}}})
     )
    )) {isSetParameters=true, module=SetVariables_#4}
   &ScaleColor(
    Scale Alpha=&FloatFromCurve(FloatCurve=DataInterface() {propertyValues {Curve {keys [
         {interpMode=RCIM_Cubic, leaveTangentWeight=0.883, tangentMode=RCTM_Break, time=0.3, value=1}
         {interpMode=RCIM_Cubic, tangentMode=RCTM_Break, time=1, value=0}
        ]}}})
    ScaleA=true
    ScaleRGB=true
   )
   &Collision(
    Age Colliding Particles=false
    Collision Enabled=&SetBoolByFloatComparison(A=$Particles.NormalizedAge, B=0.35)
    GPU Collision Type=::NewEnumerator2
    KillOnCollision=false
    Module.Control Rest State Via Surface Normal=false
    Module.EnableMaxCollisionCount=false
    Module.Randomize Collision Normal Vector=false
    Restitution=0.4
   )
   &VectorNoiseForce(
    Force Amount=&MakeCustomFloatFromBool(Bool=$Particles.HasCollided, True Float=0, False Float=2000)
    Module.Use Random Noise Time Interval=false
   )
   &GravityForce()
   &SolveForcesAndVelocity(Clamp Velocity=false, Limit Acceleration=false)
   &DynamicMaterialParameters(
    Index 0 Param 1=0.9
    Module.Index0Param0Write=true
    Module.Index0Param1Write=false
    Module.Index0Param2Write=false
    Module.Index0Param3Write=false
   )
  ]
  rendererClasses [&NiagaraMeshRendererProperties, &NiagaraSpriteRendererProperties]
  renderers [
   {rendererClass=&NiagaraMeshRendererProperties, rendererIndex=0}
   {rendererClass=&NiagaraSpriteRendererProperties, rendererIndex=1}
   {rendererClass=&NiagaraSpriteRendererProperties, rendererIndex=2}
   {rendererClass=&NiagaraSpriteRendererProperties, rendererIndex=3}
  ]
 }
 Streaks {
  simTarget=GPUComputeSim
  emitterUpdateScript [
   &EmitterState()
   &SpawnRate(
    SpawnRate=&FloatFromCurve(
     FloatCurve=DataInterface() {propertyValues {Curve {keys [
         {interpMode=RCIM_Cubic, leaveTangentWeight=0.7508, tangentMode=RCTM_User, time=0, value=1}
         {arriveTangent=-1.8497, interpMode=RCIM_Cubic, leaveTangent=-1.8497, tangentMode=RCTM_User, time=1, value=0}
        ]}}}
     CurveIndex=$Emitter.NormalizedLoopAge
     Scale Curve=150
    )
    Use Spawn Probability=false
   )
  ]
  particleSpawnScript [
   &InitializeParticle(
    Color Maximum=<rgba 0.244 1.6596 3 1>
    Color Minimum=<rgba 0.0813 0.5532 1 1>
    Color Mode=::NewEnumerator2
    Lifetime Max=0.5
    Lifetime Min=0.25
    Lifetime Mode=::NewEnumerator1
    Sprite Size=<1.6 45>
    Sprite Size Mode=::NewEnumerator4
    UsePositionOffset=false
    Write Material Random=true
   )
   &ShapeLocation(
    Box Midpoint=<0 0.5 0.5>
    Box Size=<0 250 100>
    MaskSpawnGroup=true
    Offset=$System.Range Control
    Offset Coordinate Space=::Local
    Offset Mode=::NewEnumerator0
    Override Seed=false
    Shape Primitive=::NewEnumerator2
    Surface Only=true
    UseOwnerScale=true
   )
   &AddVelocity(Velocity=<-600 0 0>)
   &CameraOffset(Camera Offset Amount=100)
  ]
  particleUpdateScript [
   &ParticleState()
   &AccelerationForce(Acceleration=&TransformVector(Vector=<-3000 0 0>, Destination Space=::Simulation))
   &SolveForcesAndVelocity(Clamp Velocity=false, Limit Acceleration=false)
   &ScaleColor(
    Scale Alpha=&FloatFromCurve(FloatCurve=DataInterface() {propertyValues {Curve {keys [{arriveTangent=-2.5374, interpMode=RCIM_Cubic, time=0, value=1}, <at 1 0 RCIM_Cubic>]}}})
    ScaleA=true
    ScaleRGB=true
   )
   &DynamicMaterialParameters(
    Index 0 Param 1=0.9
    Module.Index0Param0Write=true
    Module.Index0Param1Write=false
    Module.Index0Param2Write=false
    Module.Index0Param3Write=false
   )
  ]
  rendererClasses [&NiagaraSpriteRendererProperties]
  renderers [{rendererClass=&NiagaraSpriteRendererProperties}]
 }
}
summary {
 systemName=NS_Vehicle_Transition_Burst_v2
 emitters [
  {
   emitterName=Digital_Noise
   simTarget=GPUComputeSim
   rendererClasses [&NiagaraMeshRendererProperties, &NiagaraSpriteRendererProperties]
  }
  {emitterName=Streaks, simTarget=GPUComputeSim, rendererClasses [&NiagaraSpriteRendererProperties]}
  {
   emitterName=Digital_Noise_WorldSpace
   simTarget=GPUComputeSim
   rendererClasses [&NiagaraMeshRendererProperties, &NiagaraSpriteRendererProperties]
  }
 ]
 userVariables [
  User.Loaded {type=NiagaraBool}
  User.SystemLifetime {type=NiagaraFloat, defaultValue=0.5, description=""}
 ]
}
