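# Niagara effect for a vehicle loading transition, assembling a vehicle's skeletal mesh out of glowing cubes.
# GPU cubes lerp from a random sphere to sampled mesh points through animated region masks, with grid arrangement.
# Sweeping masks along X drive cube scale and brightness, alongside shed cubes, streaks and a moving light.
# User.Loaded kills main cubes and stops trail spawning; the mobile main emitter targets a box instead of the mesh.
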NiagaraSystem /Game/Effects/Fort_Effects/Gameplay/Vehicles/NS_Vehicle_Transition_Loading_v4
dependencies {
 usedDataInterfaces [&NiagaraDataInterfaceCurve, &NiagaraDataInterfaceSkeletalMesh]
 usedDynamicInputs []
 usedModules [
  &self:SystemSpawnScript.SystemScriptSource.SystemScriptGraph.NiagaraNodeAssignment_0.SetVariables_#1
  &SystemState
 ]
}
emitters {
 Light {
  emitterUpdateScript [
   &EmitterState()
   &SpawnBurst_Instantaneous(Use Spawn Probability=false, UseLoopCountLimit=true)
  ]
  particleSpawnScript [&InitializeParticle(Color=<rgba 0.244 1.6596 3 1>, Lifetime=0, UsePositionOffset=false, Write Material Random=true)]
  particleUpdateScript [
   &ParticleState(Kill Particles When Lifetime Has Elapsed=false, Let Infinitely Lived Particles Die When Emitter Deactivates=true)
   /SetVariables(Particles.Position=&AddVectorToPosition(
     Position=$Particles.Initial.Position
     Vector=&MakeVector(X=&Lerp_Float(A=300, B=-300, Alpha=$Emitter.NormalizedLoopAge), Y=0, Z=-50)
    )) {isSetParameters=true}
   /SetVariables(Particles.LightRendererColor:LinearColor=&MultiplyLinearColor(Linear Color=$Particles.Initial.Color, Scale Factor=<1 1 1 25>)) {isSetParameters=true}
  ]
  rendererClasses [&NiagaraComponentRendererProperties, &NiagaraLightRendererProperties]
  renderers [
   {rendererClass=&NiagaraComponentRendererProperties, rendererIndex=0}
   {rendererClass=&NiagaraLightRendererProperties, rendererIndex=1}
  ]
 }
 Main_Pulsing_Cubes {
  simTarget=GPUComputeSim
  emitterUpdateScript [
   &EmitterState()
   &SpawnBurst_Instantaneous(Spawn Count=3000, Use Spawn Probability=false, UseLoopCountLimit=true)
  ]
  particleSpawnScript [
   &InitializeParticle(
    Lifetime=0
    Mesh Scale Mode=::NewEnumerator1
    Mesh Uniform Scale Max=6
    Mesh Uniform Scale Min=3
    UsePositionOffset=false
    Write Material Random=true
   )
   &SkeletalMeshLocation(
    MaskSpawnGroup=false
    Mesh Sampling Type=::NewEnumerator7
    Offset Position Along Sampled Normal=0
    OffsetAlongNormal=true
    Orientation Sampling=::NewEnumerator1
    OverrideRandomSeed=false
    Skeletal Mesh=$System.Skeletal Mesh
   )
   &ShapeLocation(MaskSpawnGroup=false, Override Seed=false, Sphere Radius=500, UseOwnerScale=true)
   /SetVariables(Particles.NoisePosition:NiagaraPosition=$Particles.Position) {isSetParameters=true, module=SetVariables_#2}
  ]
  particleUpdateScript [
   &ParticleState(Kill Particles When Lifetime Has Elapsed=false, Let Infinitely Lived Particles Die When Emitter Deactivates=true)
   &Timeline(Play Rate=2.5)
   &RegionMask(
    Invert Region=true
    Offset:Vector3f=<-300 0 0>
    Position:Vector3f=$Particles.SkeletalMeshLocation.SampledPosition
    Region Dimensions:Vector3f=&MakeVector(
     X=&Lerp_Float(
      A=2000
      B=0
      Alpha=&FloatFromCurve(
       FloatCurve=DataInterface() {propertyValues {Curve {keys [{arriveTangent=0.4047, leaveTangentWeight=0.3469, time=0, value=0}, {arriveTangentWeight=0.3469, time=1, value=1}]}}}
       CurveIndex=$Particles.Timeline.NormalizedTime
      )
     )
     Y=1500
     Z=1500
    )
    Region Falloff=1
    Region Type:ENiagaraDebugDrawShape=::NewEnumerator1
    UseOffset=true
    UseRotation=false
   )
   /SetVariables(Particles.Position=&Lerp_Vector(
     A=Linked() {linkedVariable {name=Particles.NoisePosition, type=NiagaraPosition}}
     B=$Particles.Initial.SkeletalMeshLocation.SampledPosition
     Alpha=$Particles.RegionMask.Float
    )) {isSetParameters=true, module=SetVariables_#3}
   &self:GridArrangement(GridSize=18) {module=GridArrangement}
   &DynamicMaterialParameters(
    Index 0 Param 1=&Multiply_Float(
     A=0.0625
     B=&MakeFloatFromInt(Int=&RandomRangeInt(Minimum=0, Maximum=15))
    )
    Index 0 Param 2=&Multiply_Float(
     A=0.0625
     B=&MakeFloatFromInt(Int=&RandomRangeInt(Minimum=0, Maximum=15))
    )
    Module.Index0Param0Write=true
    Module.Index0Param1Write=false
    Module.Index0Param2Write=false
    Module.Index0Param3Write=false
   )
   &RegionMask(
    Invert Region=true
    Region Dimensions:Vector3f=<1500 800 800>
    Region Falloff:NiagaraFloat=&Lerp_Float(A=1, B=0.5, Alpha=$Particles.RegionMask.Float)
    Region Type:ENiagaraDebugDrawShape=::NewEnumerator1
    UseOffset=false
    UseRotation=false
   ) {module=RegionMask001}
   &ScaleMeshSize(Scale Factor=&VectorFromFloat(Value=&FloatFromCurve(
      FloatCurve=DataInterface() {propertyValues {Curve {keys [
          {arriveTangent=-3.2169, interpMode=RCIM_Cubic, leaveTangent=-3.2169, leaveTangentWeight=1.1396, tangentMode=RCTM_User, time=0, value=1}
          {arriveTangentWeight=1.1396, interpMode=RCIM_Cubic, tangentMode=RCTM_User, time=1, value=0}
         ]}}}
      CurveIndex=$Particles.RegionMask001.Float
     )))
   &RegionMask(
    Invert Region=true
    Offset:Vector3f=&Lerp_Vector(
     A=<-300 0 0>
     B=<300 0 0>
     Alpha=&FloatFromCurve(
      FloatCurve=DataInterface() {propertyValues {Curve {keys [{tangentMode=RCTM_User, time=0, value=1}, {tangentMode=RCTM_User, time=1, value=0}]}}}
      CurveIndex=$Emitter.NormalizedLoopAge
     )
    )
    Region Dimensions:Vector3f=<250 800 500>
    Region Falloff=0.75
    Region Type:ENiagaraDebugDrawShape=::NewEnumerator1
    UseOffset=true
    UseRotation=false
   ) {module=RegionMask002}
   &CameraOffset(Camera Offset Amount=&Multiply_Float(
     A=100
     B=&Lerp_Float(
      A=1
      B=&FloatFromCurve(
       FloatCurve=DataInterface() {propertyValues {Curve {keys [{tangentMode=RCTM_User, time=0, value=1}, {tangentMode=RCTM_User, time=1, value=0}]}}}
       CurveIndex=$Particles.RegionMask002.Float
      )
      Alpha=&FloatFromCurve(
       FloatCurve=DataInterface() {propertyValues {Curve {keys [<at 0.4 0 RCIM_Cubic>, {interpMode=RCIM_Cubic, leaveTangent=6.4224, time=0.5, value=1}]}}}
       CurveIndex=$Emitter.Age
      )
     )
    ))
   &ScaleMeshSize(Scale Factor=&VectorFromFloat(Value=&Lerp_Float(
      A=1
      B=&FloatFromCurve(
       FloatCurve=DataInterface() {propertyValues {Curve {keys [{tangentMode=RCTM_User, time=0, value=1}, {tangentMode=RCTM_User, time=1, value=0}]}}}
       CurveIndex=$Particles.RegionMask002.Float
      )
      Alpha=&FloatFromCurve(
       FloatCurve=DataInterface() {propertyValues {Curve {keys [<at 0.4 0 RCIM_Cubic>, {interpMode=RCIM_Cubic, leaveTangent=6.4224, time=0.5, value=1}]}}}
       CurveIndex=$Emitter.Age
      )
     ))) {module=ScaleMeshSize001}
   &ScaleColor(
    Scale RGB=&VectorFromFloat(Value=&RandomRangeFloat(Minimum=0.5, Maximum=2))
    ScaleA=false
    ScaleRGB=true
   )
   &RegionMask(
    Invert Region=true
    Offset:Vector3f=&Lerp_Vector(
     A=<-300 0 0>
     B=<300 0 0>
     Alpha=&FloatFromCurve(
      FloatCurve=DataInterface() {propertyValues {Curve {keys [{arriveTangent=-2.6451, time=0, value=1}, {arriveTangentWeight=1.1276, time=1, value=0}]}}}
      CurveIndex=$Emitter.NormalizedLoopAge
     )
    )
    Region Dimensions:Vector3f=<50 800 500>
    Region Falloff=0.25
    Region Type:ENiagaraDebugDrawShape=::NewEnumerator1
    UseOffset=true
    UseRotation=false
   ) {module=RegionMask003}
   &ScaleColor(
    Scale RGB=&VectorFromFloat(Value=&Lerp_Float(
      A=1
      B=10
      Alpha=&FloatFromCurve(
       FloatCurve=DataInterface() {propertyValues {Curve {keys [{arriveTangent=-2.9713, interpMode=RCIM_Cubic, leaveTangent=-2.9713, tangentMode=RCTM_User, time=0, value=1}, <at 1 0 RCIM_Cubic>]}}}
       CurveIndex=$StackContext.RegionMask003.Float
      )
     ))
    ScaleA=true
    ScaleRGB=true
   ) {module=ScaleColor001}
   &KillParticles(Kill Particles=$User.Loaded)
  ]
  rendererClasses [&NiagaraMeshRendererProperties]
  renderers [{rendererClass=&NiagaraMeshRendererProperties, rendererIndex=0}]
 }
 Mobile_Main_Pulsing_Elements {
  simTarget=GPUComputeSim
  emitterUpdateScript [
   &EmitterState()
   &SpawnBurst_Instantaneous(Spawn Count=3000, Use Spawn Probability=false, UseLoopCountLimit=true)
  ]
  particleSpawnScript [
   &InitializeParticle(
    Lifetime=0
    Mesh Scale Mode=::NewEnumerator1
    Mesh Uniform Scale Max=6
    Mesh Uniform Scale Min=3
    UsePositionOffset=false
    Write Material Random=true
   )
   &ShapeLocation(MaskSpawnGroup=false, Override Seed=false, Sphere Radius=1500, Sphere Surface Distribution=1, UseOwnerScale=true)
   /SetVariables(Particles.NoisePosition:NiagaraPosition=$Particles.Position) {isSetParameters=true, module=SetVariables_#2}
  ]
  particleUpdateScript [
   &ParticleState(Kill Particles When Lifetime Has Elapsed=false, Let Infinitely Lived Particles Die When Emitter Deactivates=true)
   &Timeline(Play Rate=2.5)
   &ShapeLocation(
    Box Midpoint=<0.45 0.5 0.4>
    Box Size=<450 275 150>
    MaskSpawnGroup=false
    Override Seed=false
    Shape Primitive=::NewEnumerator2
    Surface Only=false
    UseOwnerScale=true
   ) {module=ShapeLocation001}
   &RegionMask(
    Invert Region=true
    Offset:Vector3f=<-300 0 0>
    Position:Vector3f=$Particles.SkeletalMeshLocation.SampledPosition
    Region Dimensions:Vector3f=&MakeVector(
     X=&Lerp_Float(
      A=2000
      B=0
      Alpha=&FloatFromCurve(
       FloatCurve=DataInterface() {propertyValues {Curve {keys [{arriveTangent=0.4047, leaveTangentWeight=0.3469, time=0, value=0}, {arriveTangentWeight=0.3469, time=1, value=1}]}}}
       CurveIndex=$Particles.Timeline.NormalizedTime
      )
     )
     Y=1500
     Z=1500
    )
    Region Falloff=1
    Region Type:ENiagaraDebugDrawShape=::NewEnumerator1
    UseOffset=true
    UseRotation=false
   )
   /SetVariables(Particles.Position=&Lerp_Vector(
     A=Linked() {linkedVariable {name=Particles.NoisePosition, type=NiagaraPosition}}
     B=Linked() {linkedVariable {name=Particles.Position, type=NiagaraPosition}}
     Alpha=$Particles.RegionMask.Float
    )) {isSetParameters=true, module=SetVariables_#3}
   &self:GridArrangement(GridSize=18) {module=GridArrangement}
   &DynamicMaterialParameters(
    Index 0 Param 1=&Multiply_Float(
     A=0.0625
     B=&MakeFloatFromInt(Int=&RandomRangeInt(Minimum=0, Maximum=15))
    )
    Index 0 Param 2=&Multiply_Float(
     A=0.0625
     B=&MakeFloatFromInt(Int=&RandomRangeInt(Minimum=0, Maximum=15))
    )
    Module.Index0Param0Write=true
    Module.Index0Param1Write=false
    Module.Index0Param2Write=false
    Module.Index0Param3Write=false
   )
   &RegionMask(
    Invert Region=true
    Region Dimensions:Vector3f=<1500 800 800>
    Region Falloff:NiagaraFloat=&Lerp_Float(A=1, B=0.5, Alpha=$Particles.RegionMask.Float)
    Region Type:ENiagaraDebugDrawShape=::NewEnumerator1
    UseOffset=false
    UseRotation=false
   ) {module=RegionMask001}
   &ScaleMeshSize(Scale Factor=&VectorFromFloat(Value=&FloatFromCurve(
      FloatCurve=DataInterface() {propertyValues {Curve {keys [
          {arriveTangent=-3.2169, interpMode=RCIM_Cubic, leaveTangent=-3.2169, leaveTangentWeight=1.1396, tangentMode=RCTM_User, time=0, value=1}
          {arriveTangentWeight=1.1396, interpMode=RCIM_Cubic, tangentMode=RCTM_User, time=1, value=0}
         ]}}}
      CurveIndex=$Particles.RegionMask001.Float
     )))
   &RegionMask(
    Invert Region=true
    Offset:Vector3f=&Lerp_Vector(
     A=<-300 0 0>
     B=<300 0 0>
     Alpha=&FloatFromCurve(
      FloatCurve=DataInterface() {propertyValues {Curve {keys [{tangentMode=RCTM_User, time=0, value=1}, {tangentMode=RCTM_User, time=1, value=0}]}}}
      CurveIndex=$Emitter.NormalizedLoopAge
     )
    )
    Region Dimensions:Vector3f=<250 800 500>
    Region Falloff=0.75
    Region Type:ENiagaraDebugDrawShape=::NewEnumerator1
    UseOffset=true
    UseRotation=false
   ) {module=RegionMask002}
   &CameraOffset(Camera Offset Amount=&Multiply_Float(
     A=100
     B=&Lerp_Float(
      A=1
      B=&FloatFromCurve(
       FloatCurve=DataInterface() {propertyValues {Curve {keys [{tangentMode=RCTM_User, time=0, value=1}, {tangentMode=RCTM_User, time=1, value=0}]}}}
       CurveIndex=$Particles.RegionMask002.Float
      )
      Alpha=&FloatFromCurve(
       FloatCurve=DataInterface() {propertyValues {Curve {keys [<at 0.4 0 RCIM_Cubic>, {interpMode=RCIM_Cubic, leaveTangent=6.4224, time=0.5, value=1}]}}}
       CurveIndex=$Emitter.Age
      )
     )
    ))
   &ScaleMeshSize(Scale Factor=&VectorFromFloat(Value=&Lerp_Float(
      A=1
      B=&FloatFromCurve(
       FloatCurve=DataInterface() {propertyValues {Curve {keys [{tangentMode=RCTM_User, time=0, value=1}, {tangentMode=RCTM_User, time=1, value=0}]}}}
       CurveIndex=$Particles.RegionMask002.Float
      )
      Alpha=&FloatFromCurve(
       FloatCurve=DataInterface() {propertyValues {Curve {keys [<at 0.4 0 RCIM_Cubic>, {interpMode=RCIM_Cubic, leaveTangent=6.4224, time=0.5, value=1}]}}}
       CurveIndex=$Emitter.Age
      )
     ))) {module=ScaleMeshSize001}
   &ScaleColor(
    Scale RGB=&VectorFromFloat(Value=&RandomRangeFloat(Minimum=0.5, Maximum=2))
    ScaleA=false
    ScaleRGB=true
   )
   &RegionMask(
    Invert Region=true
    Offset:Vector3f=&Lerp_Vector(
     A=<-300 0 0>
     B=<300 0 0>
     Alpha=&FloatFromCurve(
      FloatCurve=DataInterface() {propertyValues {Curve {keys [{arriveTangent=-2.6451, time=0, value=1}, {arriveTangentWeight=1.1276, time=1, value=0}]}}}
      CurveIndex=$Emitter.NormalizedLoopAge
     )
    )
    Region Dimensions:Vector3f=<50 800 500>
    Region Falloff=0.25
    Region Type:ENiagaraDebugDrawShape=::NewEnumerator1
    UseOffset=true
    UseRotation=false
   ) {module=RegionMask003}
   &ScaleColor(
    Scale RGB=&VectorFromFloat(Value=&Lerp_Float(
      A=1
      B=10
      Alpha=&FloatFromCurve(
       FloatCurve=DataInterface() {propertyValues {Curve {keys [{arriveTangent=-2.9713, interpMode=RCIM_Cubic, leaveTangent=-2.9713, tangentMode=RCTM_User, time=0, value=1}, <at 1 0 RCIM_Cubic>]}}}
       CurveIndex=$StackContext.RegionMask003.Float
      )
     ))
    ScaleA=true
    ScaleRGB=true
   ) {module=ScaleColor001}
   &KillParticles(Kill Particles=$User.Loaded)
  ]
  rendererClasses [&NiagaraMeshRendererProperties]
  renderers [{rendererClass=&NiagaraMeshRendererProperties, rendererIndex=0}]
 }
 Mobile_World_Space_Cubes_Trailing {
  simTarget=GPUComputeSim
  emitterUpdateScript [
   &EmitterState()
   &SpawnRate(
    SpawnRate=&FloatFromCurve(
     FloatCurve=DataInterface() {propertyValues {Curve {keys [{arriveTangent=-3.0234, interpMode=RCIM_Cubic, leaveTangent=-3.0234, tangentMode=RCTM_User, time=0, value=1}, <at 1 0 RCIM_Cubic>]}}}
     CurveIndex=$Emitter.NormalizedLoopAge
     Scale Curve=&MakeCustomFloatFromBool(
      Bool=$User.Loaded
      True Float=0
      False Float=&MakeCustomFloatFromBool(Bool=&SetBoolByFloatComparison(A=$Emitter.Age, B=0.4), True Float=2000, False Float=0)
     )
    )
    Use Spawn Probability=false
   )
  ]
  particleSpawnScript [
   &InitializeParticle(
    Lifetime Max=1.25
    Lifetime Min=0.75
    Lifetime Mode=::NewEnumerator1
    Mesh Scale Mode=::NewEnumerator1
    Mesh Uniform Scale Max=4
    Sprite Size Mode=::NewEnumerator3
    UsePositionOffset=false
    Write Material Random=true
   )
   /SetVariables(Particles.Flare Size:Vector2f=&Vector2DFromFloat(Value=&Multiply_Float(
      A=&Multiply_Float(
       A=&RemapRange(
        Input Value To Remap=&MakeFloatFromVector(Vector=$Particles.Scale, Channel=::NewEnumerator0)
        Input Min=1
        Input Max=4
        Output Min=1.2
        Output Max=3
       )
       B=1
      )
      B=&MakeFloatFromVector2d(Vector 2D=$Particles.SpriteSize)
     ))) {isSetParameters=true, module=SetVariables_#4}
   &RegionMask(
    Invert Region=true
    Offset:Vector3f=&Lerp_Vector(
     A=<-300 0 0>
     B=<300 0 0>
     Alpha=&FloatFromCurve(
      FloatCurve=DataInterface() {propertyValues {Curve {keys [{tangentMode=RCTM_User, time=0, value=1}, {tangentMode=RCTM_User, time=1, value=0}]}}}
      CurveIndex=$Emitter.NormalizedLoopAge
     )
    )
    Region Dimensions:Vector3f=<250 800 500>
    Region Falloff=0.75
    Region Type:ENiagaraDebugDrawShape=::NewEnumerator1
    UseOffset=true
    UseRotation=false
   )
   &KillParticles(Kill Particles=$Particles.RegionMask.IsInsideMask)
   &AddVelocity(Velocity=<-450 0 0>)
   &InitialMeshOrientation(Mesh Orientation Mode=::NewEnumerator2, Orientation Coordinate Space=::Local, Use Rotation Vector=false)
  ]
  particleUpdateScript [
   &ParticleState()
   &SkeletalMeshLocation(
    MaskSpawnGroup=false
    Orientation Sampling=::NewEnumerator1
    OverrideRandomSeed=false
    Position Sampling=::NewEnumerator1
    Skeletal Mesh=$System.Skeletal Mesh
    Source Movement Scale:Vector3f=<0 0 0>
   )
   &DynamicMaterialParameters(
    Index 0 Param 1=&Multiply_Float(
     A=0.0625
     B=&MakeFloatFromInt(Int=&RandomRangeInt(Minimum=0, Maximum=15))
    )
    Index 0 Param 2=&Multiply_Float(
     A=0.0625
     B=&MakeFloatFromInt(Int=&RandomRangeInt(Minimum=0, Maximum=15))
    )
    Module.Index0Param0Write=true
    Module.Index0Param1Write=true
    Module.Index0Param2Write=false
    Module.Index0Param3Write=false
   )
   &ScaleMeshSize(Scale Factor=&VectorFromFloat(Value=&FloatFromCurve(FloatCurve=DataInterface() {propertyValues {Curve {keys [<at 0.55 1 RCIM_Cubic>,<at 1 0 RCIM_Cubic>]}}})))
   &CurlNoiseForce(Noise Strength=1500, Pan Noise=false)
   &InheritSourceMovement(
    Applied Position Scale=&VectorFromFloat(Value=&FloatFromCurve(FloatCurve=DataInterface() {propertyValues {Curve {keys [{arriveTangent=-1.5533, interpMode=RCIM_Cubic, leaveTangent=-1.5533, tangentMode=RCTM_User, time=0, value=1}, <at 1 0 RCIM_Cubic>]}}}))
    ApplyForce=false
    ApplyPosition=true
    ApplyPositionFalloff=false
   )
   &SolveForcesAndVelocity(Clamp Velocity=false, Limit Acceleration=false)
   &CameraOffset(Camera Offset Amount=100)
   /SetVariables(Particles.SpriteColor=&MultiplyLinearColor(
     Linear Color=<rgba 0 0.5809 1 0.1>
     Scale Factor=&MakeVector4(
      X=1
      Y=1
      Z=1
      W=&FloatFromCurve(FloatCurve=DataInterface() {propertyValues {Curve {keys [{interpMode=RCIM_Cubic, tangentMode=RCTM_Break, time=0.3, value=1}, {interpMode=RCIM_Cubic, tangentMode=RCTM_Break, time=1, value=0}]}}})
     )
    )) {isSetParameters=true, module=SetVariables_#5}
   &ScaleColor(
    Scale RGB=&VectorFromFloat(Value=&RandomRangeFloat(Minimum=0.5, Maximum=2))
    ScaleA=true
    ScaleRGB=true
   )
  ]
  rendererClasses [&NiagaraMeshRendererProperties, &NiagaraSpriteRendererProperties]
  renderers [
   {rendererClass=&NiagaraMeshRendererProperties, rendererIndex=0}
   {rendererClass=&NiagaraSpriteRendererProperties, rendererIndex=1}
  ]
 }
 Streaks {
  simTarget=GPUComputeSim
  emitterUpdateScript [
   &EmitterState()
   &SpawnRate(
    SpawnRate=&FloatFromCurve(
     FloatCurve=DataInterface() {propertyValues {Curve {keys [
         {interpMode=RCIM_Cubic, leaveTangentWeight=0.7508, tangentMode=RCTM_User, time=0, value=1}
         {arriveTangent=-1.8497, interpMode=RCIM_Cubic, leaveTangent=-1.8497, tangentMode=RCTM_User, time=1, value=0}
        ]}}}
     CurveIndex=$Emitter.NormalizedLoopAge
     Scale Curve=&MakeCustomFloatFromBool(
      Bool=$User.Loaded
      True Float=0
      False Float=&MakeCustomFloatFromBool(Bool=&SetBoolByFloatComparison(A=$Emitter.Age, B=0.4), True Float=100, False Float=0)
     )
    )
    Use Spawn Probability=false
   )
  ]
  particleSpawnScript [
   &InitializeParticle(
    Color Maximum=<rgba 0.244 1.6596 3 1>
    Color Minimum=<rgba 0.0813 0.5532 1 1>
    Color Mode=::NewEnumerator2
    Lifetime Max=0.5
    Lifetime Min=0.25
    Lifetime Mode=::NewEnumerator1
    Sprite Size=<1.6 45>
    Sprite Size Mode=::NewEnumerator4
    UsePositionOffset=false
    Write Material Random=true
   )
   &ShapeLocation(
    Box Size=<0 200 125>
    MaskSpawnGroup=true
    Offset=&Lerp_Vector(
     A=<-300 0 0>
     B=<300 0 0>
     Alpha=&FloatFromCurve(
      FloatCurve=DataInterface() {propertyValues {Curve {keys [{tangentMode=RCTM_User, time=0, value=1}, {tangentMode=RCTM_User, time=1, value=0}]}}}
      CurveIndex=$Emitter.NormalizedLoopAge
     )
    )
    Offset Coordinate Space=::Local
    Offset Mode=::NewEnumerator0
    Override Seed=false
    Shape Primitive=::NewEnumerator2
    Surface Only=true
    UseOwnerScale=true
   )
   &AddVelocity(Velocity=<-600 0 50>)
   &CameraOffset(Camera Offset Amount=100)
  ]
  particleUpdateScript [
   &ParticleState()
   &AccelerationForce(Acceleration=&TransformVector(Vector=<-1500 0 0>, Destination Space=::Simulation))
   &SolveForcesAndVelocity(Clamp Velocity=false, Limit Acceleration=false)
   &ScaleColor(
    Scale Alpha=&FloatFromCurve(FloatCurve=DataInterface() {propertyValues {Curve {keys [{arriveTangent=-2.5374, interpMode=RCIM_Cubic, time=0, value=1}, <at 1 0 RCIM_Cubic>]}}})
    ScaleA=true
    ScaleRGB=true
   )
   &DynamicMaterialParameters(
    Index 0 Param 1=0.9
    Module.Index0Param0Write=true
    Module.Index0Param1Write=false
    Module.Index0Param2Write=false
    Module.Index0Param3Write=false
   )
  ]
  rendererClasses [&NiagaraSpriteRendererProperties]
  renderers [{rendererClass=&NiagaraSpriteRendererProperties}]
 }
 World_Space_Cubes_Epic {
  simTarget=GPUComputeSim
  emitterUpdateScript [
   &EmitterState()
   &SpawnRate(
    SpawnRate=&FloatFromCurve(
     FloatCurve=DataInterface() {propertyValues {Curve {keys [{arriveTangent=-3.0234, interpMode=RCIM_Cubic, leaveTangent=-3.0234, tangentMode=RCTM_User, time=0, value=1}, <at 1 0 RCIM_Cubic>]}}}
     CurveIndex=$Emitter.NormalizedLoopAge
     Scale Curve=&MakeCustomFloatFromBool(
      Bool=$User.Loaded
      True Float=0
      False Float=&MakeCustomFloatFromBool(Bool=&SetBoolByFloatComparison(A=$Emitter.Age, B=0.4), True Float=1000, False Float=0)
     )
    )
    Use Spawn Probability=false
   )
  ]
  particleSpawnScript [
   &InitializeParticle(
    Lifetime Max=1.25
    Lifetime Min=0.75
    Lifetime Mode=::NewEnumerator1
    Mesh Scale Mode=::NewEnumerator1
    Mesh Uniform Scale Max=4
    Sprite Size Mode=::NewEnumerator3
    UsePositionOffset=false
    Write Material Random=true
   )
   /SetVariables(Particles.Flare Size:Vector2f=&Vector2DFromFloat(Value=&Multiply_Float(
      A=&Multiply_Float(
       A=&RemapRange(
        Input Value To Remap=&MakeFloatFromVector(Vector=$Particles.Scale, Channel=::NewEnumerator0)
        Input Min=1
        Input Max=4
        Output Min=1.2
        Output Max=3
       )
       B=1
      )
      B=&MakeFloatFromVector2d(Vector 2D=$Particles.SpriteSize)
     ))) {isSetParameters=true, module=SetVariables_#4}
   &RegionMask(
    Invert Region=true
    Offset:Vector3f=&Lerp_Vector(
     A=<-300 0 0>
     B=<300 0 0>
     Alpha=&FloatFromCurve(
      FloatCurve=DataInterface() {propertyValues {Curve {keys [{tangentMode=RCTM_User, time=0, value=1}, {tangentMode=RCTM_User, time=1, value=0}]}}}
      CurveIndex=$Emitter.NormalizedLoopAge
     )
    )
    Region Dimensions:Vector3f=<250 800 500>
    Region Falloff=0.75
    Region Type:ENiagaraDebugDrawShape=::NewEnumerator1
    UseOffset=true
    UseRotation=false
   )
   &KillParticles(Kill Particles=$Particles.RegionMask.IsInsideMask)
   &AddVelocity(Velocity=<-450 0 0>)
   &InitialMeshOrientation(Mesh Orientation Mode=::NewEnumerator2, Orientation Coordinate Space=::Local, Use Rotation Vector=false)
  ]
  particleUpdateScript [
   &ParticleState()
   &SkeletalMeshLocation(
    MaskSpawnGroup=false
    Mesh Sampling Type=::NewEnumerator7
    Offset Position Along Sampled Normal=20
    OffsetAlongNormal=true
    Orientation Sampling=::NewEnumerator1
    OverrideRandomSeed=false
    Position Sampling=::NewEnumerator1
    Skeletal Mesh=$System.Skeletal Mesh
    Source Movement Scale:Vector3f=<0 0 0>
   )
   &DynamicMaterialParameters(
    Index 0 Param 1=&Multiply_Float(
     A=0.0625
     B=&MakeFloatFromInt(Int=&RandomRangeInt(Minimum=0, Maximum=15))
    )
    Index 0 Param 2=&Multiply_Float(
     A=0.0625
     B=&MakeFloatFromInt(Int=&RandomRangeInt(Minimum=0, Maximum=15))
    )
    Module.Index0Param0Write=true
    Module.Index0Param1Write=true
    Module.Index0Param2Write=false
    Module.Index0Param3Write=false
   )
   &ScaleMeshSize(Scale Factor=&VectorFromFloat(Value=&FloatFromCurve(FloatCurve=DataInterface() {propertyValues {Curve {keys [<at 0.55 1 RCIM_Cubic>,<at 1 0 RCIM_Cubic>]}}})))
   &Collision(
    Age Colliding Particles=false
    Collision Enabled=&SetBoolByFloatComparison(A=$Particles.NormalizedAge, B=0.35)
    GPU Collision Type=::NewEnumerator2
    KillOnCollision=false
    Module.Control Rest State Via Surface Normal=false
    Module.EnableMaxCollisionCount=false
    Module.Randomize Collision Normal Vector=false
    Restitution=0.4
   )
   &CurlNoiseForce(Noise Strength=1500, Pan Noise=false)
   &GravityForce()
   &SolveForcesAndVelocity(Clamp Velocity=false, Limit Acceleration=false)
   &CameraOffset(Camera Offset Amount=100)
   /SetVariables(Particles.SpriteColor=&MultiplyLinearColor(
     Linear Color=<rgba 0 0.5809 1 0.1>
     Scale Factor=&MakeVector4(
      X=1
      Y=1
      Z=1
      W=&FloatFromCurve(FloatCurve=DataInterface() {propertyValues {Curve {keys [{interpMode=RCIM_Cubic, tangentMode=RCTM_Break, time=0.3, value=1}, {interpMode=RCIM_Cubic, tangentMode=RCTM_Break, time=1, value=0}]}}})
     )
    )) {isSetParameters=true, module=SetVariables_#5}
   &ScaleColor(
    Scale RGB=&VectorFromFloat(Value=&RandomRangeFloat(Minimum=0.5, Maximum=2))
    ScaleA=true
    ScaleRGB=true
   )
  ]
  rendererClasses [&NiagaraMeshRendererProperties, &NiagaraSpriteRendererProperties]
  renderers [
   {rendererClass=&NiagaraMeshRendererProperties, rendererIndex=0}
   {rendererClass=&NiagaraSpriteRendererProperties, rendererIndex=1}
  ]
 }
 World_Space_Cubes_Trailing {
  simTarget=GPUComputeSim
  emitterUpdateScript [
   &EmitterState()
   &SpawnRate(
    SpawnRate=&FloatFromCurve(
     FloatCurve=DataInterface() {propertyValues {Curve {keys [{arriveTangent=-3.0234, interpMode=RCIM_Cubic, leaveTangent=-3.0234, tangentMode=RCTM_User, time=0, value=1}, <at 1 0 RCIM_Cubic>]}}}
     CurveIndex=$Emitter.NormalizedLoopAge
     Scale Curve=&MakeCustomFloatFromBool(
      Bool=$User.Loaded
      True Float=0
      False Float=&MakeCustomFloatFromBool(Bool=&SetBoolByFloatComparison(A=$Emitter.Age, B=0.4), True Float=1000, False Float=0)
     )
    )
    Use Spawn Probability=false
   )
  ]
  particleSpawnScript [
   &InitializeParticle(
    Lifetime Max=1.25
    Lifetime Min=0.75
    Lifetime Mode=::NewEnumerator1
    Mesh Scale Mode=::NewEnumerator1
    Mesh Uniform Scale Max=4
    Sprite Size Mode=::NewEnumerator3
    UsePositionOffset=false
    Write Material Random=true
   )
   /SetVariables(Particles.Flare Size:Vector2f=&Vector2DFromFloat(Value=&Multiply_Float(
      A=&Multiply_Float(
       A=&RemapRange(
        Input Value To Remap=&MakeFloatFromVector(Vector=$Particles.Scale, Channel=::NewEnumerator0)
        Input Min=1
        Input Max=4
        Output Min=1.2
        Output Max=3
       )
       B=1
      )
      B=&MakeFloatFromVector2d(Vector 2D=$Particles.SpriteSize)
     ))) {isSetParameters=true, module=SetVariables_#4}
   &RegionMask(
    Invert Region=true
    Offset:Vector3f=&Lerp_Vector(
     A=<-300 0 0>
     B=<300 0 0>
     Alpha=&FloatFromCurve(
      FloatCurve=DataInterface() {propertyValues {Curve {keys [{tangentMode=RCTM_User, time=0, value=1}, {tangentMode=RCTM_User, time=1, value=0}]}}}
      CurveIndex=$Emitter.NormalizedLoopAge
     )
    )
    Region Dimensions:Vector3f=<250 800 500>
    Region Falloff=0.75
    Region Type:ENiagaraDebugDrawShape=::NewEnumerator1
    UseOffset=true
    UseRotation=false
   )
   &KillParticles(Kill Particles=$Particles.RegionMask.IsInsideMask)
   &AddVelocity(Velocity=<-450 0 0>)
   &InitialMeshOrientation(Mesh Orientation Mode=::NewEnumerator2, Orientation Coordinate Space=::Local, Use Rotation Vector=false)
  ]
  particleUpdateScript [
   &ParticleState()
   &SkeletalMeshLocation(
    MaskSpawnGroup=false
    Mesh Sampling Type=::NewEnumerator7
    Offset Position Along Sampled Normal=20
    OffsetAlongNormal=true
    Orientation Sampling=::NewEnumerator1
    OverrideRandomSeed=false
    Position Sampling=::NewEnumerator1
    Skeletal Mesh=$System.Skeletal Mesh
    Source Movement Scale:Vector3f=<0 0 0>
   )
   &DynamicMaterialParameters(
    Index 0 Param 1=&Multiply_Float(
     A=0.0625
     B=&MakeFloatFromInt(Int=&RandomRangeInt(Minimum=0, Maximum=15))
    )
    Index 0 Param 2=&Multiply_Float(
     A=0.0625
     B=&MakeFloatFromInt(Int=&RandomRangeInt(Minimum=0, Maximum=15))
    )
    Module.Index0Param0Write=true
    Module.Index0Param1Write=true
    Module.Index0Param2Write=false
    Module.Index0Param3Write=false
   )
   &ScaleMeshSize(Scale Factor=&VectorFromFloat(Value=&FloatFromCurve(FloatCurve=DataInterface() {propertyValues {Curve {keys [<at 0.55 1 RCIM_Cubic>,<at 1 0 RCIM_Cubic>]}}})))
   &CurlNoiseForce(Noise Strength=1500, Pan Noise=false)
   &InheritSourceMovement(
    Applied Position Scale=&VectorFromFloat(Value=&FloatFromCurve(FloatCurve=DataInterface() {propertyValues {Curve {keys [{arriveTangent=-1.5533, interpMode=RCIM_Cubic, leaveTangent=-1.5533, tangentMode=RCTM_User, time=0, value=1}, <at 1 0 RCIM_Cubic>]}}}))
    ApplyForce=false
    ApplyPosition=true
    ApplyPositionFalloff=false
   )
   &SolveForcesAndVelocity(Clamp Velocity=false, Limit Acceleration=false)
   &CameraOffset(Camera Offset Amount=100)
   /SetVariables(Particles.SpriteColor=&MultiplyLinearColor(
     Linear Color=<rgba 0 0.5809 1 0.1>
     Scale Factor=&MakeVector4(
      X=1
      Y=1
      Z=1
      W=&FloatFromCurve(FloatCurve=DataInterface() {propertyValues {Curve {keys [{interpMode=RCIM_Cubic, tangentMode=RCTM_Break, time=0.3, value=1}, {interpMode=RCIM_Cubic, tangentMode=RCTM_Break, time=1, value=0}]}}})
     )
    )) {isSetParameters=true, module=SetVariables_#5}
   &ScaleColor(
    Scale RGB=&VectorFromFloat(Value=&RandomRangeFloat(Minimum=0.5, Maximum=2))
    ScaleA=true
    ScaleRGB=true
   )
  ]
  rendererClasses [&NiagaraMeshRendererProperties, &NiagaraSpriteRendererProperties]
  renderers [
   {rendererClass=&NiagaraMeshRendererProperties, rendererIndex=0}
   {rendererClass=&NiagaraSpriteRendererProperties, rendererIndex=1}
  ]
 }
}
summary {
 systemName=NS_Vehicle_Transition_Loading_v4
 emitters [
  {
   emitterName=World_Space_Cubes_Epic
   simTarget=GPUComputeSim
   rendererClasses [&NiagaraMeshRendererProperties, &NiagaraSpriteRendererProperties]
  }
  {emitterName=Main_Pulsing_Cubes, simTarget=GPUComputeSim, rendererClasses [&NiagaraMeshRendererProperties]}
  {emitterName=Streaks, simTarget=GPUComputeSim, rendererClasses [&NiagaraSpriteRendererProperties]}
  {
   emitterName=World_Space_Cubes_Trailing
   simTarget=GPUComputeSim
   rendererClasses [&NiagaraMeshRendererProperties, &NiagaraSpriteRendererProperties]
  }
  {emitterName=Mobile_Main_Pulsing_Elements, simTarget=GPUComputeSim, rendererClasses [&NiagaraMeshRendererProperties]}
  {
   emitterName=Mobile_World_Space_Cubes_Trailing
   simTarget=GPUComputeSim
   rendererClasses [&NiagaraMeshRendererProperties, &NiagaraSpriteRendererProperties]
  }
  {emitterName=Light, rendererClasses [&NiagaraComponentRendererProperties, &NiagaraLightRendererProperties]}
 ]
 userVariables [
  User.Loaded {type=NiagaraBool}
  User.SystemLifetime {type=NiagaraFloat, defaultValue=0.5, description=""}
 ]
}
